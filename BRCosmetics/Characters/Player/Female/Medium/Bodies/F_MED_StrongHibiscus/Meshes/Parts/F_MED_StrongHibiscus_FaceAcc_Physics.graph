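# Physics asset for the F_MED_StrongHibiscus face accessory mesh, set up for the motion of left and right ponytails on this female character.
# Each ponytail is a capsule chain from a MSTR root through five segments, linked by constraints with swing and narrow twist limits that differ slightly per side.
# Unconstrained capsules on the spine, neck, clavicles and upper-arm twist bones approximate the upper body as collision shapes.
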PhysicsAsset /BRCosmetics/Characters/Player/Female/Medium/Bodies/F_MED_StrongHibiscus/Meshes/Parts/F_MED_StrongHibiscus_FaceAcc_Physics
bodies {
 clavicle_l [{shapeType=Capsule, length=22.5, radius=7.5, shapeName=clavicle_l_capsule, center=<10.4812 0 -1.3785>, rotation=<rot -90 0 0>}]
 clavicle_r [{
   shapeType=Capsule
   length=22.5
   radius=7
   shapeName=clavicle_r_capsule
   center=<-9.7668 1.1705 1.411>
   rotation=<rot 70.5172 -179.7892 178.7375>
  }]
 dyn_l_ponytail_01 [{shapeType=Capsule, length=4, radius=2, shapeName=dyn_l_ponytail_01_capsule, center=<8.3097 -0.366 0>, rotation=<rot 90 180 -173.6902>}]
 dyn_l_ponytail_02 [{shapeType=Capsule, length=8.5, radius=2, shapeName=dyn_l_ponytail_02_capsule, center=<5 -0.7 0>, rotation=<rot 90 0 0>}]
 dyn_l_ponytail_03 [{shapeType=Capsule, length=8.5, radius=2, shapeName=dyn_l_ponytail_03_capsule, center=<5 -0.7 0>, rotation=<rot 90 0 0>}]
 dyn_l_ponytail_04 [{shapeType=Capsule, length=9, radius=1.5, shapeName=dyn_l_ponytail_04_capsule, center=<5 -0.47 0>, rotation=<rot 90 0 0>}]
 dyn_l_ponytail_05 [{shapeType=Capsule, length=4.5, radius=3.5, shapeName=dyn_l_ponytail_05_capsule, center=<5 -1.7 0>, rotation=<rot 90 0 0>}]
 dyn_l_ponytail_MSTR [{shapeType=Capsule, length=1, radius=0.5, shapeName=dyn_l_ponytail_MSTR_capsule, rotation=<rot 90 0 0>}]
 dyn_r_ponytail_01 [{shapeType=Capsule, length=3.5, radius=2, shapeName=dyn_r_ponytail_01_capsule, center=<7.9407 -0.3789 0>, rotation=<rot 90 0 7.3421>}]
 dyn_r_ponytail_02 [{shapeType=Capsule, length=8.5, radius=2, shapeName=dyn_r_ponytail_02_capsule, center=<5 -0.85 0>, rotation=<rot 90 0 0>}]
 dyn_r_ponytail_03 [{shapeType=Capsule, length=8.5, radius=2, shapeName=dyn_r_ponytail_03_capsule, center=<5 -0.55 0>, rotation=<rot 90 0 0>}]
 dyn_r_ponytail_04 [{shapeType=Capsule, length=9, radius=1.5, shapeName=dyn_r_ponytail_04_capsule, center=<5 -0.4 0>, rotation=<rot 90 0 0>}]
 dyn_r_ponytail_05 [{shapeType=Capsule, length=4.5, radius=3.5, shapeName=dyn_r_ponytail_05_capsule, center=<5 -1.7 -1>, rotation=<rot 90 0 0>}]
 dyn_r_ponytail_MSTR [{shapeType=Capsule, length=1, radius=0.5, shapeName=dyn_r_ponytail_MSTR_capsule, rotation=<rot 90 0 0>}]
 neck_01 [{shapeType=Capsule, length=23, radius=7, shapeName=neck_01_capsule, center=<0.3988 4.6399 0>}]
 spine_01 [{shapeType=Capsule, length=4, radius=11.5, shapeName=spine_01_capsule, center=<3.1784 4.1116 0>}]
 spine_03 [{shapeType=Capsule, length=3.5, radius=16, shapeName=spine_03_capsule, center=<3.6936 7.0885 0>}]
 spine_05 [{shapeType=Capsule, length=7.5, radius=15.5, shapeName=spine_05_capsule, center=<-1.9306 5.6965 0>}]
 upperarm_twist_01_l [{shapeType=Capsule, length=18, radius=7, shapeName=upperarm_twist_01_l_capsule, center=<5.4688 0.8284 -0.6657>, rotation=<rot 90 0 0>}]
 upperarm_twist_01_r [{shapeType=Capsule, length=15, radius=6, shapeName=upperarm_twist_01_r_capsule, center=<-4.2493 0 0>, rotation=<rot 90 0 0>}]
}
constraints [
 {bone1Name=dyn_l_ponytail_02, bone2Name=dyn_l_ponytail_01, swing1LimitDegrees=40, swing2LimitDegrees=20, twistLimitDegrees=5}
 {bone1Name=dyn_l_ponytail_03, bone2Name=dyn_l_ponytail_02, swing1LimitDegrees=35, swing2LimitDegrees=35, twistLimitDegrees=5}
 {bone1Name=dyn_l_ponytail_04, bone2Name=dyn_l_ponytail_03, swing1LimitDegrees=35, swing2LimitDegrees=35, twistLimitDegrees=5}
 {bone1Name=dyn_l_ponytail_05, bone2Name=dyn_l_ponytail_04, swing1LimitDegrees=40, swing2LimitDegrees=40, twistLimitDegrees=5}
 {bone1Name=dyn_r_ponytail_02, bone2Name=dyn_r_ponytail_01, swing1LimitDegrees=40, swing2LimitDegrees=20, twistLimitDegrees=10}
 {bone1Name=dyn_r_ponytail_03, bone2Name=dyn_r_ponytail_02, swing1LimitDegrees=25, swing2LimitDegrees=25, twistLimitDegrees=5}
 {bone1Name=dyn_r_ponytail_04, bone2Name=dyn_r_ponytail_03, swing1LimitDegrees=25, swing2LimitDegrees=25, twistLimitDegrees=5}
 {bone1Name=dyn_r_ponytail_05, bone2Name=dyn_r_ponytail_04, swing1LimitDegrees=40, swing2LimitDegrees=40, twistLimitDegrees=5}
 {bone1Name=dyn_l_ponytail_01, bone2Name=dyn_l_ponytail_MSTR, swing1LimitDegrees=40, swing2LimitDegrees=25, twistLimitDegrees=5}
 {bone1Name=dyn_r_ponytail_01, bone2Name=dyn_r_ponytail_MSTR, swing1LimitDegrees=40, swing2LimitDegrees=25, twistLimitDegrees=5}
]
