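# Control Rig for the StrongHibiscus face accessory part that offsets the ponytail and left earring bones as the head and upper body move.
# Spherical pose readers on head and spine_05 output weights for Modify Transforms nodes, run as two Sequence branches: ponytails, then left earring.
# Earring weights are multiplied by the "is altskin on" variable, so those offsets vanish at its default of zero.
# Three further pose readers are unconnected, and RB_Enabled is unused in the graph.
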ControlRigBlueprint /Game/Characters/Player/Female/Medium/Bodies/F_MED_StrongHibiscus/Meshes/Parts/F_MED_StrongHibiscus_FaceAcc_CtrlRig
graphs [
 RigVMFunctionLibrary {links [], nodes []}
 RigVMModel {nodes [
   RigUnit_BeginExecution = &RigUnit_BeginExecution()
   RigVMFunction_Sequence = &RigVMFunction_Sequence(ExecuteContext=@RigUnit_BeginExecution.ExecutePin)
   n1 = &RigUnit_SphericalPoseReader(
    ActiveRegionScaleFactors=(PositiveWidth=1,NegativeWidth=1,PositiveHeight=1,NegativeHeight=1)
    ActiveRegionSize=0.04
    Debug=(bDrawDebug=false,bDraw2D=False,bDrawLocalAxes=False,DebugScale=25,DebugSegments=20,DebugThickness=0.25)
    DriverAxis=(X=1,Y=0,Z=0)
    DriverItem=(Type=Bone,Name="head")
    FalloffRegionScaleFactors=(PositiveWidth=1,NegativeWidth=1,PositiveHeight=1,NegativeHeight=1)
    FalloffSize=0.2
    FlipHeightScaling=False
    FlipWidthScaling=False
    OptionalParentItem=(Type=Bone,Name="None")
    RotationOffset=(X=90,Y=0,Z=-130)
    ExecutePin=@RigVMFunction_Sequence.A
   )
   > &RigUnit_ModifyTransforms(
    ItemToModify=((Item=(Type=Bone,Name="dyn_l_ponytail_MSTR"),Transform=(Rotation=(X=-0.2588,Y=0,Z=0,W=0.9659),Translation=(X=0,Y=1,Z=2.3),Scale3D=(X=1,Y=1,Z=1))))
    WeightMinimum=0
    Weight=@n1.OutputParam
   )
   > &RigUnit_ModifyTransforms(
    ItemToModify=((Item=(Type=Bone,Name="dyn_r_ponytail_MSTR"),Transform=(Rotation=(X=-0.342,Y=0,Z=0,W=0.9397),Translation=(X=0,Y=1,Z=2.3),Scale3D=(X=1,Y=1,Z=1))))
    WeightMinimum=0
    Weight=@n1.OutputParam
   )
   > n2 = &RigUnit_SphericalPoseReader(
    ActiveRegionScaleFactors=(PositiveWidth=1,NegativeWidth=1,PositiveHeight=1,NegativeHeight=1)
    ActiveRegionSize=0.1
    Debug=(bDrawDebug=false,bDraw2D=False,bDrawLocalAxes=False,DebugScale=25,DebugSegments=20,DebugThickness=0.25)
    DriverAxis=(X=0,Y=0,Z=1)
    DriverItem=(Type=Bone,Name="spine_05")
    FalloffRegionScaleFactors=(PositiveWidth=1,NegativeWidth=1,PositiveHeight=1,NegativeHeight=1)
    FalloffSize=0.3
    FlipHeightScaling=False
    FlipWidthScaling=False
    OptionalParentItem=(Type=Bone,Name="head")
    RotationOffset=(X=60,Y=0,Z=0)
   )
   > &RigUnit_ModifyTransforms(
    ItemToModify=((Item=(Type=Bone,Name="dyn_l_ponytail_MSTR"),Transform=(Rotation=(X=0,Y=0,Z=-0.5736,W=0.8192),Translation=(X=1,Y=-1,Z=0),Scale3D=(X=1,Y=1,Z=1))))
    WeightMinimum=0
    Weight=@n2.OutputParam
   )
   > &RigUnit_ModifyTransforms(
    ItemToModify=((Item=(Type=Bone,Name="dyn_r_ponytail_MSTR"),Transform=(Rotation=(X=0,Y=0,Z=-0.5,W=0.866),Translation=(X=0,Y=0,Z=0),Scale3D=(X=1,Y=1,Z=1))))
    WeightMinimum=0
    Weight=@n2.OutputParam
   )
   > n3 = &RigUnit_SphericalPoseReader(
    ActiveRegionScaleFactors=(PositiveWidth=1,NegativeWidth=1,PositiveHeight=1,NegativeHeight=1)
    ActiveRegionSize=0.1
    Debug=(bDrawDebug=false,bDraw2D=False,bDrawLocalAxes=False,DebugScale=25,DebugSegments=20,DebugThickness=0.25)
    DriverAxis=(X=0,Y=0,Z=-1)
    DriverItem=(Type=Bone,Name="spine_05")
    FalloffRegionScaleFactors=(PositiveWidth=1,NegativeWidth=1,PositiveHeight=1,NegativeHeight=1)
    FalloffSize=0.3
    FlipHeightScaling=False
    FlipWidthScaling=False
    OptionalParentItem=(Type=Bone,Name="head")
    RotationOffset=(X=105,Y=0,Z=0)
   )
   > &RigUnit_ModifyTransforms(
    ItemToModify=((Item=(Type=Bone,Name="dyn_l_ponytail_MSTR"),Transform=(Rotation=(X=0,Y=0,Z=0.5,W=0.866),Translation=(X=0,Y=0,Z=0),Scale3D=(X=1,Y=1,Z=1))))
    WeightMinimum=0
    Weight=@n3.OutputParam
   )
   > &RigUnit_ModifyTransforms(
    ItemToModify=((Item=(Type=Bone,Name="dyn_r_ponytail_MSTR"),Transform=(Rotation=(X=0,Y=0,Z=0.5736,W=0.8192),Translation=(X=-1,Y=-2,Z=0),Scale3D=(X=1,Y=1,Z=1))))
    WeightMinimum=0
    Weight=@n3.OutputParam
   )
   n4 = &RigUnit_SphericalPoseReader(
    ActiveRegionScaleFactors=(PositiveWidth=1,NegativeWidth=1,PositiveHeight=1,NegativeHeight=1)
    ActiveRegionSize=0.04
    Debug=(bDrawDebug=false,bDraw2D=False,bDrawLocalAxes=False,DebugScale=25,DebugSegments=20,DebugThickness=0.25)
    DriverAxis=(X=0,Y=1,Z=0)
    DriverItem=(Type=Bone,Name="head")
    FalloffRegionScaleFactors=(PositiveWidth=1,NegativeWidth=1,PositiveHeight=1,NegativeHeight=1)
    FalloffSize=0.2
    FlipHeightScaling=False
    FlipWidthScaling=False
    OptionalParentItem=(Type=Bone,Name="spine_05")
    RotationOffset=(X=110,Y=0,Z=-45)
    ExecutePin=@RigVMFunction_Sequence.B
   )
   Multiply = &RigVMFunction_MathDoubleMul(A=@n4.OutputParam, B=RigVMVariableNode() {variable=is altskin on}.Value)
   > &RigUnit_ModifyTransforms(
    ItemToModify=((Item=(Type=Bone,Name="dyn_l_earring_MSTR"),Transform=(Rotation=(X=0.0342,Y=-0.0489,Z=0.1438,W=0.9878),Translation=(X=0,Y=0,Z=0),Scale3D=(X=1,Y=1,Z=1))))
    WeightMinimum=0
    Weight=@Multiply.Result
   )
   > n5 = &RigUnit_SphericalPoseReader(
    ActiveRegionScaleFactors=(PositiveWidth=1,NegativeWidth=1,PositiveHeight=1.5,NegativeHeight=1)
    ActiveRegionSize=0.1
    Debug=(bDrawDebug=false,bDraw2D=False,bDrawLocalAxes=False,DebugScale=25,DebugSegments=20,DebugThickness=0.25)
    DriverAxis=(X=0,Y=1,Z=0)
    DriverItem=(Type=Bone,Name="head")
    FalloffRegionScaleFactors=(PositiveWidth=1,NegativeWidth=1,PositiveHeight=1.5,NegativeHeight=1)
    FalloffSize=0.2
    FlipHeightScaling=False
    FlipWidthScaling=False
    OptionalParentItem=(Type=Bone,Name="spine_05")
    RotationOffset=(X=110,Y=0,Z=-10)
   )
   > &RigUnit_ModifyTransforms(
    ItemToModify=((Item=(Type=Bone,Name="dyn_l_earring_MSTR"),Transform=(Rotation=(X=0.0786,Y=-0.0423,Z=0.1512,W=0.9845),Translation=(X=0,Y=0,Z=0),Scale3D=(X=1,Y=1,Z=1))))
    WeightMinimum=0
    Weight=&RigVMFunction_MathDoubleMul(A=@n5.OutputParam, B=RigVMVariableNode() {variable=is altskin on}.Value).Result
   )
   > n6 = &RigUnit_SphericalPoseReader(
    ActiveRegionScaleFactors=(PositiveWidth=1,NegativeWidth=1,PositiveHeight=1,NegativeHeight=1)
    ActiveRegionSize=0.04
    Debug=(bDrawDebug=false,bDraw2D=False,bDrawLocalAxes=False,DebugScale=25,DebugSegments=20,DebugThickness=0.25)
    DriverAxis=(X=0,Y=1,Z=0)
    DriverItem=(Type=Bone,Name="head")
    FalloffRegionScaleFactors=(PositiveWidth=1,NegativeWidth=1,PositiveHeight=1,NegativeHeight=1)
    FalloffSize=0.2
    FlipHeightScaling=False
    FlipWidthScaling=False
    OptionalParentItem=(Type=Bone,Name="spine_05")
    RotationOffset=(X=125,Y=0,Z=20)
   )
   > &RigUnit_ModifyTransforms(
    ItemToModify=((Item=(Type=Bone,Name="dyn_l_earring_MSTR"),Transform=(Rotation=(X=0.1757,Y=-0.1168,Z=0.1759,W=0.9615),Translation=(X=0,Y=0,Z=0),Scale3D=(X=1,Y=1,Z=1))))
    WeightMinimum=0
    Weight=&RigVMFunction_MathDoubleMul(A=@n6.OutputParam, B=RigVMVariableNode() {variable=is altskin on}.Value).Result
   )
   &RigUnit_SphericalPoseReader(
    ActiveRegionScaleFactors=(PositiveWidth=0.3,NegativeWidth=0.3,PositiveHeight=1,NegativeHeight=1)
    ActiveRegionSize=0.1
    Debug=(bDrawDebug=false,bDraw2D=False,bDrawLocalAxes=False,DebugScale=25,DebugSegments=20,DebugThickness=0.25)
    DriverAxis=(X=-1,Y=0,Z=0)
    DriverItem=(Type=Bone,Name="clavicle_r")
    ExecutePin=()
    FalloffRegionScaleFactors=(PositiveWidth=1,NegativeWidth=1,PositiveHeight=1,NegativeHeight=1)
    FalloffSize=0.2
    FlipHeightScaling=False
    FlipWidthScaling=False
    OptionalParentItem=(Type=Bone,Name="head")
    RotationOffset=(X=-110,Y=0,Z=0)
   )
   &RigUnit_SphericalPoseReader(
    ActiveRegionScaleFactors=(PositiveWidth=0.3,NegativeWidth=0.3,PositiveHeight=1,NegativeHeight=1)
    ActiveRegionSize=0.1
    Debug=(bDrawDebug=true,bDraw2D=False,bDrawLocalAxes=False,DebugScale=25,DebugSegments=20,DebugThickness=0.25)
    DriverAxis=(X=1,Y=0,Z=0)
    DriverItem=(Type=Bone,Name="clavicle_l")
    ExecutePin=()
    FalloffRegionScaleFactors=(PositiveWidth=1,NegativeWidth=1,PositiveHeight=1,NegativeHeight=1)
    FalloffSize=0.2
    FlipHeightScaling=False
    FlipWidthScaling=False
    OptionalParentItem=(Type=Bone,Name="head")
    RotationOffset=(X=65,Y=0,Z=0)
   )
   &RigUnit_SphericalPoseReader(
    ActiveRegionScaleFactors=(PositiveWidth=1,NegativeWidth=1,PositiveHeight=1,NegativeHeight=1)
    ActiveRegionSize=0.04
    Debug=(bDrawDebug=true,bDraw2D=False,bDrawLocalAxes=False,DebugScale=25,DebugSegments=20,DebugThickness=0.25)
    DriverAxis=(X=0,Y=1,Z=0)
    DriverItem=(Type=Bone,Name="head")
    ExecutePin=()
    FalloffRegionScaleFactors=(PositiveWidth=1,NegativeWidth=1,PositiveHeight=1,NegativeHeight=1)
    FalloffSize=0.2
    FlipHeightScaling=False
    FlipWidthScaling=False
    OptionalParentItem=(Type=Bone,Name="spine_05")
    RotationOffset=(X=110,Y=0,Z=-10)
   )
  ]}
]
hierarchy [
 root {loc=<0 0 0>, rot=<0 0 0 1>}
 pelvis {parent=root, loc=<0 -0.5301 98.0571>, rot=<0.0127 0.707 -0.0127 -0.707>}
 spine_01 {parent=pelvis, loc=<6.1208 -0.9328 0>, rot=<0 0 0.034 -0.9994>}
 spine_02 {parent=spine_01, loc=<6.5996 -0.2095 0>, rot=<0 0 0.0238 -0.9997>}
 spine_03 {parent=spine_02, loc=<6.4914 -0.1237 0>, rot=<0 0 0.0149 -0.9999>}
 spine_04 {parent=spine_03, loc=<6.9598 0.0794 0>, rot=<0 0 -0.0132 -0.9999>}
 spine_05 {parent=spine_04, loc=<6.9987 0 0>, rot=<0 0 -0.0015 -1>}
 neck_01 {parent=spine_05, loc=<11.3138 0.1529 0>, rot=<0 0 -0.2003 -0.9797>}
 neck_02 {parent=neck_01, loc=<4.5627 0 0>, rot=<0 0 0.0504 -0.9987>}
 head {parent=neck_02, loc=<5.3914 0 0>, rot=<0 0 0.115 -0.9934>}
 dyn_l_ponytail_MSTR {parent=head, loc=<-1.0523 -3.723 -3.6091>, rot=<0.0029 -0.7071 -0.0029 -0.7071>}
 dyn_l_ponytail_01 {parent=dyn_l_ponytail_MSTR, loc=<0 0 0>, rot=<0.3047 -0.6476 0.1324 -0.6857>}
 dyn_l_ponytail_02 {parent=dyn_l_ponytail_01, loc=<9.6253 0 0>, rot=<-0.0094 0.0217 0.0061 -0.9997>}
 dyn_l_ponytail_03 {parent=dyn_l_ponytail_02, loc=<10.6183 0 0>, rot=<-0.0134 0.0469 -0.083 -0.9954>}
 dyn_l_ponytail_04 {parent=dyn_l_ponytail_03, loc=<10.659 0 0>, rot=<0.0111 -0.0164 -0.1228 -0.9922>}
 dyn_l_ponytail_05 {parent=dyn_l_ponytail_04, loc=<11.0091 0 0>, rot=<-0.0029 -0.0694 -0.042 -0.9967>}
 dyn_r_ponytail_MSTR {parent=head, loc=<-1.0525 -3.723 3.6091>, rot=<0.0029 -0.7071 -0.0029 -0.7071>}
 dyn_r_ponytail_01 {parent=dyn_r_ponytail_MSTR, loc=<0 0 0>, rot=<0.1324 -0.6857 0.3047 -0.6476>}
 dyn_r_ponytail_02 {parent=dyn_r_ponytail_01, loc=<9.6249 -0.0003 0>, rot=<0.0094 -0.0217 0.0061 -0.9997>}
 dyn_r_ponytail_03 {parent=dyn_r_ponytail_02, loc=<10.6187 0.0002 0>, rot=<0.0134 -0.0469 -0.083 -0.9954>}
 dyn_r_ponytail_04 {parent=dyn_r_ponytail_03, loc=<10.6587 -0.0001 0>, rot=<-0.0111 0.0164 -0.1228 -0.9922>}
 dyn_r_ponytail_05 {parent=dyn_r_ponytail_04, loc=<11.009 0 0>, rot=<0.0029 0.0694 -0.042 -0.9967>}
 dyn_l_earring_MSTR {parent=head, loc=<0.1893 0.0903 -6.168>, rot=<0.0029 -0.7071 -0.0029 -0.7071>}
 dyn_l_earring {parent=dyn_l_earring_MSTR, loc=<0 0 0>, rot=<0.3859 -0.5925 -0.3859 -0.5925>}
 dyn_r_earring_MSTR {parent=head, loc=<0.189 0.0903 6.168>, rot=<0.0029 -0.7071 -0.0029 -0.7071>}
 dyn_r_earring {parent=dyn_r_earring_MSTR, loc=<0 0 0>, rot=<0.5935 -0.3843 -0.5935 -0.3843>}
 dyn_hair_strand_a_MSTR {parent=head, loc=<9.6568 6.8758 4.4057>, rot=<0.0029 -0.7071 -0.0029 -0.7071>}
 dyn_hair_strand_a_01 {parent=dyn_hair_strand_a_MSTR, loc=<0 0 0>, rot=<-0.4402 0.3051 0.8314 -0.1482>}
 dyn_hair_strand_a_02 {parent=dyn_hair_strand_a_01, loc=<5.9677 0 0>, rot=<0.04 -0.2164 0.1961 -0.9556>}
 dyn_hair_strand_b_MSTR {parent=head, loc=<10.9872 5.0804 4.6116>, rot=<0.0029 -0.7071 -0.0029 -0.7071>}
 dyn_hair_strand_b_01 {parent=dyn_hair_strand_b_MSTR, loc=<0 0 0>, rot=<-0.2466 0.3398 0.9029 -0.0918>}
 dyn_hair_strand_b_02 {parent=dyn_hair_strand_b_01, loc=<5.1378 0.456 -1.8142>, rot=<0.0137 -0.3005 0.1358 -0.944>}
 dyn_hair_a_MSTR {parent=head, loc=<13.0697 10.8864 -1.4869>, rot=<0.0029 -0.7071 -0.0029 -0.7071>}
 dyn_hair_a_01 {parent=dyn_hair_a_MSTR, loc=<0 0 0>, rot=<0.6049 -0.3662 -0.6049 -0.3662>}
 dyn_hair_a_02 {parent=dyn_hair_a_01, loc=<8.7586 -2.7158 0.7104>, rot=<0 0 0 1>}
 dyn_hair_b_MSTR {parent=head, loc=<13.214 7.6287 -3.7775>, rot=<0.0029 -0.7071 -0.0029 -0.7071>}
 dyn_hair_b_01 {parent=dyn_hair_b_MSTR, loc=<0 0 0>, rot=<0.6049 -0.3662 -0.6049 -0.3662>}
 dyn_hair_b_02 {parent=dyn_hair_b_01, loc=<9.0084 -2.7111 -6.5907>, rot=<-0.1577 -0.0403 -0.0064 -0.9866>}
 dyn_hair_b_03 {parent=dyn_hair_b_01, loc=<6.9295 -2.0164 -9.5713>, rot=<0 -0.3178 0 -0.9481>}
 dyn_curl_a {parent=head, loc=<15.7605 7.1355 -1.282>, rot=<0.4313 -0.2658 -0.4804 -0.7159>}
 dyn_curl_b {parent=head, loc=<10.9835 0.1247 -8.5959>, rot=<0.7544 -0.5035 -0.2439 -0.3434>}
 dyn_curl_c {parent=head, loc=<13.8213 3.3304 -7.0765>, rot=<0.6619 -0.6257 -0.3661 -0.1904>}
 clavicle_l {parent=spine_05, loc=<5.4831 -0.1053 -3.2946>, rot=<0.0465 -0.7289 -0.009 -0.6829>}
 upperarm_l {parent=clavicle_l, loc=<9.8658 -1.2136 -0.4105>, rot=<0.0164 -0.3245 -0.0269 -0.9454>}
 upperarm_twist_01_l {parent=upperarm_l, loc=<7.2135 -0.1134 0>, rot=<0 0 0 1>}
 clavicle_r {parent=spine_05, loc=<5.4831 -0.1053 3.2946>, rot=<0.7289 0.0465 -0.6829 0.009>}
 upperarm_r {parent=clavicle_r, loc=<-9.8658 1.2136 0.4105>, rot=<0.0164 -0.3245 -0.0269 -0.9454>}
 upperarm_twist_01_r {parent=upperarm_r, loc=<-7.2135 0.1134 0>, rot=<0 0 0 -1>}
]
variables [
 RB_Enabled {type=bool, default=False}
 is altskin on {type=double, default=0}
]
