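# Anim Blueprint for the StrongHibiscus face accessory part, adding secondary motion to hair, curls, earrings and ponytails.
# It copies the parent mesh pose through a Control Rig, adjusts upper-arm bones, then runs AnimDynamics on hair, curls and earrings.
# The cached result feeds either a RigidBody simulation or, when back bling is worn or rigid bodies are disabled, constrained, offset, AnimDynamics-driven ponytails.
# The EventGraph updates that switch, gravity and arm socket transforms.
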AnimBlueprint /BRCosmetics/Characters/Player/Female/Medium/Bodies/F_MED_StrongHibiscus/Meshes/Parts/F_MED_StrongHibiscus_FaceAcc_AnimBP
target_skeleton=/Game/Characters/Player/Female/Medium/Bodies/F_MED_StrongHibiscus/Meshes/Parts/F_MED_StrongHibiscus_FaceAcc_Skeleton
graphs [
 AnimGraph {nodes [
   n1 = Anim Dynamics - Bone: dyn_l_earring(Alpha=1, ExternalForce="0,0,0") {
    angularDampingOverride=0.99
    angularSpringConstant=1000
    bAngularSpring=true
    bGravityOverrideInSimSpace=true
    bOverrideAngularDamping=true
    bUseGravityOverride=true
    alphaBoolBlend {blendOption=Linear}
    boundBone {boneName=dyn_l_earring}
    componentAppliedLinearAccClamp=<3000 3000 3000>
    componentLinearAccScale=<0.3 0.3 0.3>
    componentLinearVelScale=<0.3 0.3 0.3>
    gravityOverride {z=-980}
    physicsBodyDefinitions [{
      boundBone {boneName=dyn_l_earring}
      boxExtents=<12 12 12>
      constraintSetup {angularLimitsMax=<5 10 10>, angularLimitsMin=<-5 -20 -70>, angularTarget=<0 0 0>}
      localJointOffset=<3 0 0>
     }]
   }
   Anim Dynamics - Bone: dyn_r_earring(ComponentPose=@n1, Alpha=1, ExternalForce="0,0,0") {
    angularDampingOverride=0.99
    angularSpringConstant=1000
    bAngularSpring=true
    bGravityOverrideInSimSpace=true
    bOverrideAngularDamping=true
    bUseGravityOverride=true
    bUseSphericalLimits=true
    alphaBoolBlend {blendOption=Linear}
    boundBone {boneName=dyn_r_earring}
    componentAppliedLinearAccClamp=<3000 3000 3000>
    componentLinearAccScale=<0.3 0.3 0.3>
    componentLinearVelScale=<0.3 0.3 0.3>
    gravityOverride {z=-980}
    physicsBodyDefinitions [{
      collisionType=CustomSphere
      sphereCollisionRadius=2
      boundBone {boneName=dyn_r_earring}
      boxExtents=<12 12 12>
      constraintSetup {angularLimitsMax=<5 10 70>, angularLimitsMin=<-5 -20 -10>, angularTarget=<0 0 0>}
      localJointOffset=<3 0 0>
     }]
    sphericalLimits [
     {limitRadius=7.3, drivingBone {boneName=upperarm_r}, sphereLocalOffset=<-3 0 0>}
     {limitRadius=5, drivingBone {boneName=head}, sphereLocalOffset=<0 0 0>}
    ]
   }
   n2 = RigidBody!(
    ComponentPose=Local To Component(LocalPose=Use cached pose 'AnimDyn_Pose'())
    Alpha
    OverrideWorldGravity:Vector=Get Gravity Override()
    ExternalForce
   ) -> (Pose) {
    bClampLinearTranslationLimitToRefPose=true
    bOverrideWorldGravity=true
    overridePhysicsAsset=&/F_MED_StrongHibiscus_FaceAcc_Physics
    simulationSpace=BaseBoneSpace
    alphaBoolBlend {blendOption=Linear}
    baseBoneRef {boneName=spine_05}
    componentAppliedLinearAccClamp=<5000 5000 5000>
    componentLinearAccScale=<0.15 0.15 0.2>
    componentLinearVelScale=<0.16 0.16 0.22>
    overrideWorldGravity {z=-980}
    simSpaceSettings {
     maxAngularAcceleration=150
     maxAngularVelocity=5
     maxLinearAcceleration=200
     maxLinearVelocity=200
     worldAlpha=1
     externalLinearDragV=<0.7 0.7 0.7>
    }
   }
   Output Pose(Result=Blend Poses by bool(
     BlendPose_0=Use cached pose 'WithBack_Pose'()
     BlendPose_1=Component To Local(ComponentPose=@n2)
     bActiveValue=Get WithoutRB() -> (WithoutRB:bool)
    ) {blendType=Linear}) {layerGroup=None}
   n3 = NOT Boolean(A=Get bIsInFrontEnd())
   n4 = "Constraint - dyn_l_ponytail_01 (PP)"!(
    ComponentPose=Local To Component(LocalPose=Use cached pose 'AnimDyn_Pose'())
    bAlphaBoolEnabled:bool=@n3
    ConstraintWeights_0=0.8
    ConstraintWeights_1:float=0.2
   ) -> (Pose) {
    alphaInputType=Bool
    alphaBoolBlend {blendOption=Linear}
    boneToModify {boneName=dyn_l_ponytail_01}
    constraintSetup [{targetBone {boneName=neck_02}}, {targetBone {boneName=clavicle_l}}]
    constraintWeights [1,1]
   }
   n5 = "Constraint - dyn_r_ponytail_01 (PP)"!(ComponentPose=@n4, bAlphaBoolEnabled:bool=@n3, ConstraintWeights_0=0.8, ConstraintWeights_1:float=0.2) -> (Pose) {
    alphaInputType=Bool
    alphaBoolBlend {blendOption=Linear}
    boneToModify {boneName=dyn_r_ponytail_01}
    constraintSetup [{targetBone {boneName=neck_02}}, {targetBone {boneName=clavicle_r}}]
    constraintWeights [1,1]
   }
   n6 = "Constraint - dyn_l_ponytail_02 (PP)"!(ComponentPose=@n5, Alpha=1, ConstraintWeights_0=0.3, ConstraintWeights_1:float=0.7) -> (Pose) {
    alphaBoolBlend {blendOption=Linear}
    boneToModify {boneName=dyn_l_ponytail_02}
    constraintSetup [{targetBone {boneName=spine_05}}, {targetBone {boneName=clavicle_l}}]
    constraintWeights [1,1]
   }
   n7 = "Constraint - dyn_r_ponytail_02 (PP)"!(ComponentPose=@n6, Alpha=1, ConstraintWeights_0=0.3, ConstraintWeights_1:float=0.7) -> (Pose) {
    alphaBoolBlend {blendOption=Linear}
    boneToModify {boneName=dyn_r_ponytail_02}
    constraintSetup [{targetBone {boneName=spine_05}}, {targetBone {boneName=clavicle_r}}]
    constraintWeights [1,1]
   }
   n8 = "Transform (Modify) Bone - Bone: dyn_l_ponytail_02"!(ComponentPose=@n7, Alpha=1, Translation="-0.378,0.000,-0.840", Rotation="-9.636,24.273,-4.423") -> (Pose) {
    rotationMode=BMM_Additive
    rotationSpace=BCS_BoneSpace
    translationMode=BMM_Additive
    translationSpace=BCS_BoneSpace
    alphaBoolBlend {blendOption=Linear}
    boneToModify {boneName=dyn_l_ponytail_02}
    rotation=<rot -9.636 24.273 -4.423>
    translation {x=-0.378, z=-0.84}
   }
   n9 = "Transform (Modify) Bone - Bone: dyn_r_ponytail_02"!(ComponentPose=@n8, Alpha=1, Translation="-0.984,-1.017,-0.103", Rotation="1.295,11.342,-3.683") -> (Pose) {
    rotationMode=BMM_Additive
    rotationSpace=BCS_BoneSpace
    translationMode=BMM_Additive
    translationSpace=BCS_BoneSpace
    alphaBoolBlend {blendOption=Linear}
    boneToModify {boneName=dyn_r_ponytail_02}
    rotation=<rot 1.295 11.342 -3.683>
    translation=<-0.984 -1.017 -0.103>
   }
   n10 = "Constraint - dyn_l_ponytail_03 (PP)"!(ComponentPose=@n9, Alpha=1, ConstraintWeights_0=0.8, ConstraintWeights_1:float=0.2) -> (Pose) {
    alphaBoolBlend {blendOption=Linear}
    boneToModify {boneName=dyn_l_ponytail_03}
    constraintSetup [{targetBone {boneName=spine_04}}, {targetBone {boneName=clavicle_l}}]
    constraintWeights [1,1]
   }
   n11 = "Constraint - dyn_r_ponytail_03 (PP)"!(ComponentPose=@n10, Alpha=1, ConstraintWeights_0=0.8, ConstraintWeights_1:float=0.2) -> (Pose) {
    alphaBoolBlend {blendOption=Linear}
    boneToModify {boneName=dyn_r_ponytail_03}
    constraintSetup [{targetBone {boneName=spine_04}}, {targetBone {boneName=clavicle_r}}]
    constraintWeights [1,1]
   }
   n12 = "Transform (Modify) Bone - Bone: dyn_l_ponytail_03"!(ComponentPose=@n11, Alpha=1, Translation="-0.251,2.574,-2.472", Rotation="-5.139,35.730,-1.791") -> (Pose) {
    rotationMode=BMM_Additive
    rotationSpace=BCS_BoneSpace
    translationMode=BMM_Additive
    translationSpace=BCS_BoneSpace
    alphaBoolBlend {blendOption=Linear}
    boneToModify {boneName=dyn_l_ponytail_03}
    rotation=<rot -5.139 35.73 -1.791>
    translation=<-0.251 2.574 -2.472>
   }
   n13 = "Transform (Modify) Bone - Bone: dyn_r_ponytail_03"!(ComponentPose=@n12, Alpha=1, Translation="-1.357,1.705,2.124", Rotation="3.510,28.239,2.893") -> (Pose) {
    rotationMode=BMM_Additive
    rotationSpace=BCS_BoneSpace
    translationMode=BMM_Additive
    translationSpace=BCS_BoneSpace
    alphaBoolBlend {blendOption=Linear}
    boneToModify {boneName=dyn_r_ponytail_03}
    rotation=<rot 3.51 28.2391 2.893>
    translation=<-1.357 1.705 2.124>
   }
   n14 = "Constraint - dyn_l_ponytail_04 (P)"(ComponentPose=@n13, Alpha=1, ConstraintWeights_0=1) {
    alphaBoolBlend {blendOption=Linear}
    boneToModify {boneName=dyn_l_ponytail_04}
    constraintSetup [{targetBone {boneName=spine_04}}]
    constraintWeights [1]
   }
   n15 = "Constraint - dyn_r_ponytail_04 (P)"(ComponentPose=@n14, Alpha=1, ConstraintWeights_0=1) {
    alphaBoolBlend {blendOption=Linear}
    boneToModify {boneName=dyn_r_ponytail_04}
    constraintSetup [{targetBone {boneName=spine_04}}]
    constraintWeights [1]
   }
   n16 = "Transform (Modify) Bone - Bone: dyn_l_ponytail_04"!(ComponentPose=@n15, Alpha=1, Translation="-2.554,9.103,-3.760", Rotation="-0.330,21.374,-0.048") -> (Pose) {
    rotationMode=BMM_Additive
    rotationSpace=BCS_BoneSpace
    scaleMode=BMM_Replace
    translationMode=BMM_Additive
    translationSpace=BCS_BoneSpace
    alphaBoolBlend {blendOption=Linear}
    boneToModify {boneName=dyn_l_ponytail_04}
    rotation=<rot -0.33 21.374 -0.048>
    translation=<-2.554 9.103 -3.76>
   }
   n17 = "Transform (Modify) Bone - Bone: dyn_r_ponytail_04"!(ComponentPose=@n16, Alpha=1, Translation="-1.914,8.333,2.287", Rotation="1.288,22.827,30.455") -> (Pose) {
    rotationMode=BMM_Additive
    rotationSpace=BCS_BoneSpace
    translationMode=BMM_Additive
    translationSpace=BCS_BoneSpace
    alphaBoolBlend {blendOption=Linear}
    boneToModify {boneName=dyn_r_ponytail_04}
    rotation=<rot 1.288 22.827 30.455>
    translation=<-1.914 8.333 2.287>
   }
   n18 = "Transform (Modify) Bone - Bone: dyn_l_ponytail_05"!(ComponentPose=@n17, bAlphaBoolEnabled:bool=True, Rotation="0,-40,0") -> (Pose) {
    alphaInputType=Bool
    rotationMode=BMM_Additive
    rotationSpace=BCS_BoneSpace
    translationMode=BMM_Additive
    translationSpace=BCS_BoneSpace
    alphaBoolBlend {blendOption=Linear, blendOutTime=1}
    boneToModify {boneName=dyn_l_ponytail_05}
    rotation {yaw=-40}
   }
   n19 = "Transform (Modify) Bone - Bone: dyn_r_ponytail_05"!(ComponentPose=@n18, bAlphaBoolEnabled:bool=True, Rotation="0,-40,0") -> (Pose) {
    alphaInputType=Bool
    rotationMode=BMM_Additive
    rotationSpace=BCS_BoneSpace
    translationMode=BMM_Additive
    translationSpace=BCS_BoneSpace
    alphaBoolBlend {blendOption=Linear, blendOutTime=1}
    boneToModify {boneName=dyn_r_ponytail_05}
    rotation {yaw=-40}
   }
   n20 = Anim Dynamics - Bone: dyn_l_ponytail_03!(ComponentPose=@n19, bAlphaBoolEnabled:bool=True, GravityOverride:Vector="0,0,-980") -> (Pose) {
    alphaInputType=Bool
    angularDampingOverride=0.99
    angularSpringConstant=1500
    bAngularSpring=true
    bOverrideAngularDamping=true
    bUseGravityOverride=true
    bUsePlanarLimit=false
    gravityScale=0.35
    simulationSpace=RootRelative
    alphaBoolBlend {blendOption=Linear}
    boundBone {boneName=dyn_l_ponytail_03}
    componentAppliedLinearAccClamp=<1000 1000 1000>
    componentLinearAccScale=<0.6 0.5 0.5>
    componentLinearVelScale=<0.6 0.5 0.5>
    gravityOverride {z=-980}
    physicsBodyDefinitions [{
      boundBone {boneName=dyn_l_ponytail_03}
      boxExtents=<20 20 20>
      constraintSetup {angularLimitsMax=<5 30 0>, angularLimitsMin=<-5 -30 -50>, angularTarget=<1 0 0>}
      localJointOffset=<2 0.1 0>
     }]
   }
   n21 = Anim Dynamics - Bone: dyn_r_ponytail_03!(ComponentPose=@n20, bAlphaBoolEnabled:bool=True, GravityOverride:Vector="0,0,-980") -> (Pose) {
    alphaInputType=Bool
    angularDampingOverride=0.99
    angularSpringConstant=2000
    bAngularSpring=true
    bOverrideAngularDamping=true
    bUseGravityOverride=true
    bUsePlanarLimit=false
    gravityScale=0.35
    simulationSpace=RootRelative
    alphaBoolBlend {blendOption=Linear}
    boundBone {boneName=dyn_r_ponytail_03}
    componentAppliedLinearAccClamp=<1000 1000 1000>
    componentLinearAccScale=<0.6 0.5 0.5>
    componentLinearVelScale=<0.6 0.5 0.5>
    gravityOverride {z=-980}
    physicsBodyDefinitions [{
      boundBone {boneName=dyn_r_ponytail_03}
      boxExtents=<20 18 18>
      constraintSetup {angularLimitsMax=<5 30 0>, angularLimitsMin=<-5 -30 -50>, angularTarget=<1 0 0>}
      localJointOffset=<1 0.1 0>
     }]
   }
   n22 = Anim Dynamics - Bone: dyn_l_ponytail_04!(ComponentPose=@n21, bAlphaBoolEnabled:bool=True, GravityOverride:Vector="0,0,-980") -> (Pose) {
    alphaInputType=Bool
    angularDampingOverride=0.99
    angularSpringConstant=1500
    bAngularSpring=true
    bOverrideAngularDamping=true
    bUseGravityOverride=true
    bUsePlanarLimit=false
    gravityScale=0.35
    simulationSpace=RootRelative
    alphaBoolBlend {blendOption=Linear}
    boundBone {boneName=dyn_l_ponytail_04}
    componentAppliedLinearAccClamp=<1000 1000 1000>
    componentLinearAccScale=<0.6 0.5 0.5>
    componentLinearVelScale=<0.6 0.5 0.5>
    gravityOverride {z=-980}
    physicsBodyDefinitions [{
      boundBone {boneName=dyn_l_ponytail_04}
      boxExtents=<18 18 18>
      constraintSetup {angularLimitsMax=<5 20 0>, angularLimitsMin=<-5 -20 -20>, angularTarget=<1 0 0>}
      localJointOffset=<3 0.1 0>
     }]
   }
   n23 = Anim Dynamics - Bone: dyn_r_ponytail_04!(ComponentPose=@n22, bAlphaBoolEnabled:bool=True, GravityOverride:Vector="0,0,-980") -> (Pose) {
    alphaInputType=Bool
    angularDampingOverride=0.99
    angularSpringConstant=2000
    bAngularSpring=true
    bOverrideAngularDamping=true
    bUseGravityOverride=true
    bUsePlanarLimit=false
    gravityScale=0.35
    simulationSpace=RootRelative
    alphaBoolBlend {blendOption=Linear}
    boundBone {boneName=dyn_r_ponytail_04}
    componentAppliedLinearAccClamp=<1000 1000 1000>
    componentLinearAccScale=<0.6 0.5 0.5>
    componentLinearVelScale=<0.6 0.5 0.5>
    gravityOverride {z=-980}
    physicsBodyDefinitions [{
      boundBone {boneName=dyn_r_ponytail_04}
      boxExtents=<18 16 16>
      constraintSetup {angularLimitsMax=<5 20 0>, angularLimitsMin=<-5 -20 -20>, angularTarget=<1 0 0>}
      localJointOffset=<3 0.1 0>
     }]
   }
   n24 = Anim Dynamics - Bone: dyn_l_ponytail_05!(ComponentPose=@n23, Alpha=1) -> (Pose) {
    angularDampingOverride=0.99
    angularSpringConstant=2000
    bAngularSpring=true
    bOverrideAngularDamping=true
    bUseGravityOverride=true
    bUsePlanarLimit=false
    gravityScale=0.35
    simulationSpace=RootRelative
    alphaBoolBlend {blendOption=Linear}
    boundBone {boneName=dyn_l_ponytail_05}
    componentAppliedLinearAccClamp=<1000 1000 1000>
    componentLinearAccScale=<0.6 0.5 0.5>
    componentLinearVelScale=<0.6 0.5 0.5>
    gravityOverride {z=-980}
    physicsBodyDefinitions [{
      boundBone {boneName=dyn_l_ponytail_05}
      boxExtents=<15 15 15>
      constraintSetup {angularLimitsMax=<5 30 0>, angularLimitsMin=<-5 -30 -30>, angularTarget=<1 0 0>}
      localJointOffset=<1 0 0>
     }]
   }
   n25 = Anim Dynamics - Bone: dyn_r_ponytail_05!(ComponentPose=@n24, Alpha=1) -> (Pose) {
    angularDampingOverride=0.99
    angularSpringConstant=2000
    bAngularSpring=true
    bOverrideAngularDamping=true
    bUseGravityOverride=true
    bUsePlanarLimit=false
    gravityScale=0.35
    simulationSpace=RootRelative
    alphaBoolBlend {blendOption=Linear}
    boundBone {boneName=dyn_r_ponytail_05}
    componentAppliedLinearAccClamp=<1000 1000 1000>
    componentLinearAccScale=<0.6 0.5 0.5>
    componentLinearVelScale=<0.6 0.5 0.5>
    gravityOverride {z=-980}
    physicsBodyDefinitions [{
      boundBone {boneName=dyn_r_ponytail_05}
      boxExtents=<15 15 15>
      constraintSetup {angularLimitsMax=<5 30 0>, angularLimitsMin=<-5 -30 -30>, angularTarget=<1 0 0>}
      localJointOffset=<1 0 0>
     }]
   }
   Save cached pose 'WithBack_Pose'(Pose=Component To Local(ComponentPose=@n25)) {cacheName=WithBack_Pose}
   n26 = Anim Dynamics - Bone: dyn_hair_a_01(ComponentPose=Local To Component(LocalPose=Use cached pose 'Start_Pose'()), Alpha=1, ExternalForce="0,0,0") {
    angularDampingOverride=0.99
    angularSpringConstant=500
    bAngularSpring=true
    bGravityOverrideInSimSpace=true
    bOverrideAngularDamping=true
    bUseGravityOverride=true
    alphaBoolBlend {blendOption=Linear}
    boundBone {boneName=dyn_hair_a_01}
    componentAppliedLinearAccClamp=<3000 3000 3000>
    componentLinearAccScale=<0.3 0.3 0.3>
    componentLinearVelScale=<0.3 0.3 0.3>
    gravityOverride {z=-980}
    physicsBodyDefinitions [{
      boundBone {boneName=dyn_hair_a_01}
      boxExtents=<20 20 20>
      constraintSetup {angularLimitsMax=<5 5 0>, angularLimitsMin=<-5 -40 -40>, angularTarget=<0 0 0>}
      localJointOffset=<4 0 0>
     }]
   }
   n27 = Anim Dynamics - Bone: dyn_hair_a_02(ComponentPose=@n26, Alpha=1, ExternalForce="0,0,0") {
    angularDampingOverride=0.99
    angularSpringConstant=1000
    bAngularSpring=true
    bGravityOverrideInSimSpace=true
    bOverrideAngularDamping=true
    bUseGravityOverride=true
    alphaBoolBlend {blendOption=Linear}
    boundBone {boneName=dyn_hair_a_02}
    componentAppliedLinearAccClamp=<3000 3000 3000>
    componentLinearAccScale=<0.3 0.3 0.3>
    componentLinearVelScale=<0.3 0.3 0.3>
    gravityOverride {z=-980}
    physicsBodyDefinitions [{
      boundBone {boneName=dyn_hair_a_02}
      boxExtents=<15 15 15>
      constraintSetup {angularLimitsMax=<10 0 10>, angularLimitsMin=<-10 -10 -20>, angularTarget=<0 0 0>}
      localJointOffset=<1 0 0>
     }]
   }
   n28 = Anim Dynamics - Bone: dyn_hair_b_01(ComponentPose=@n27, Alpha=1, ExternalForce="0,0,0") {
    angularDampingOverride=0.99
    angularSpringConstant=800
    bAngularSpring=true
    bGravityOverrideInSimSpace=true
    bOverrideAngularDamping=true
    bUseGravityOverride=true
    alphaBoolBlend {blendOption=Linear}
    boundBone {boneName=dyn_hair_b_01}
    componentAppliedLinearAccClamp=<3000 3000 3000>
    componentLinearAccScale=<0.3 0.3 0.3>
    componentLinearVelScale=<0.3 0.3 0.3>
    gravityOverride {z=-980}
    physicsBodyDefinitions [{
      boundBone {boneName=dyn_hair_b_01}
      boxExtents=<25 25 25>
      constraintSetup {angularLimitsMax=<5 0 5>, angularLimitsMin=<-5 -50 -10>, angularTarget=<0 0 0>}
      localJointOffset=<4 -0.2 0>
     }]
   }
   n29 = Anim Dynamics - Bone: dyn_hair_b_02(ComponentPose=@n28, Alpha=1, ExternalForce="0,0,0") {
    angularDampingOverride=0.95
    angularSpringConstant=500
    bAngularSpring=true
    bGravityOverrideInSimSpace=true
    bOverrideAngularDamping=true
    bUseGravityOverride=true
    alphaBoolBlend {blendOption=Linear}
    boundBone {boneName=dyn_hair_b_02}
    componentAppliedLinearAccClamp=<3000 3000 3000>
    componentLinearAccScale=<0.3 0.3 0.3>
    componentLinearVelScale=<0.3 0.3 0.3>
    gravityOverride {z=-980}
    physicsBodyDefinitions [{
      boundBone {boneName=dyn_hair_b_02}
      boxExtents=<15 15 15>
      constraintSetup {angularLimitsMax=<5 20 5>, angularLimitsMin=<-5 -30 -10>, angularTarget=<0 0 0>}
      localJointOffset=<1 0 0>
     }]
   }
   n30 = Anim Dynamics - Bone: dyn_hair_b_03(ComponentPose=@n29, Alpha=1, ExternalForce="0,0,0") {
    angularDampingOverride=0.99
    angularSpringConstant=800
    bAngularSpring=true
    bGravityOverrideInSimSpace=true
    bOverrideAngularDamping=true
    bUseGravityOverride=true
    alphaBoolBlend {blendOption=Linear}
    boundBone {boneName=dyn_hair_b_03}
    componentAppliedLinearAccClamp=<3000 3000 3000>
    componentLinearAccScale=<0.3 0.3 0.3>
    componentLinearVelScale=<0.3 0.3 0.3>
    gravityOverride {z=-980}
    physicsBodyDefinitions [{
      boundBone {boneName=dyn_hair_b_03}
      boxExtents=<17 17 17>
      constraintSetup {angularLimitsMax=<5 20 5>, angularLimitsMin=<-5 -10 -10>, angularTarget=<0 0 0>}
      localJointOffset=<2 0 0>
     }]
   }
   n31 = Anim Dynamics - Bone: dyn_hair_strand_a_01(ComponentPose=@n30, Alpha=1, ExternalForce="0,0,0") {
    angularDampingOverride=0.99
    angularSpringConstant=1000
    bAngularSpring=true
    bGravityOverrideInSimSpace=true
    bOverrideAngularDamping=true
    bUseGravityOverride=true
    alphaBoolBlend {blendOption=Linear}
    boundBone {boneName=dyn_hair_strand_a_01}
    componentAppliedLinearAccClamp=<3000 3000 3000>
    componentLinearAccScale=<0.3 0.3 0.3>
    componentLinearVelScale=<0.3 0.3 0.3>
    gravityOverride {z=-980}
    physicsBodyDefinitions [{
      boundBone {boneName=dyn_hair_strand_a_01}
      boxExtents=<20 20 20>
      constraintSetup {angularLimitsMax=<5 0 15>, angularLimitsMin=<-5 -40 -50>, angularTarget=<0 0 0>}
      localJointOffset=<1.5 0 0>
     }]
   }
   n32 = Anim Dynamics - Bone: dyn_hair_strand_a_02(ComponentPose=@n31, Alpha=1, ExternalForce="0,0,0") {
    angularDampingOverride=0.95
    angularSpringConstant=800
    bAngularSpring=true
    bGravityOverrideInSimSpace=true
    bOverrideAngularDamping=true
    bUseGravityOverride=true
    alphaBoolBlend {blendOption=Linear}
    boundBone {boneName=dyn_hair_strand_a_02}
    componentAppliedLinearAccClamp=<3000 3000 3000>
    componentLinearAccScale=<0.3 0.3 0.3>
    componentLinearVelScale=<0.3 0.3 0.3>
    gravityOverride {z=-950}
    physicsBodyDefinitions [{
      boundBone {boneName=dyn_hair_strand_a_02}
      boxExtents=<13 13 13>
      constraintSetup {angularLimitsMax=<5 5 20>, angularLimitsMin=<-5 -20 -20>, angularTarget=<0 0 0>}
      localJointOffset=<2 0 0>
     }]
   }
   n33 = Anim Dynamics - Bone: dyn_hair_strand_b_01(ComponentPose=@n32, Alpha=1, ExternalForce="0,0,0") {
    angularDampingOverride=0.99
    angularSpringConstant=1000
    bAngularSpring=true
    bGravityOverrideInSimSpace=true
    bOverrideAngularDamping=true
    bUseGravityOverride=true
    alphaBoolBlend {blendOption=Linear}
    boundBone {boneName=dyn_hair_strand_b_01}
    componentAppliedLinearAccClamp=<3000 3000 3000>
    componentLinearAccScale=<0.3 0.3 0.3>
    componentLinearVelScale=<0.3 0.3 0.3>
    gravityOverride {z=-980}
    physicsBodyDefinitions [{
      boundBone {boneName=dyn_hair_strand_b_01}
      boxExtents=<20 20 20>
      constraintSetup {angularLimitsMax=<5 0 5>, angularLimitsMin=<-5 -40 -40>, angularTarget=<0 0 0>}
      localJointOffset=<2 -0.2 0>
     }]
   }
   n34 = Anim Dynamics - Bone: dyn_hair_strand_b_02(ComponentPose=@n33, Alpha=1, ExternalForce="0,0,0") {
    angularDampingOverride=0.95
    angularSpringConstant=800
    bAngularSpring=true
    bGravityOverrideInSimSpace=true
    bOverrideAngularDamping=true
    bUseGravityOverride=true
    alphaBoolBlend {blendOption=Linear}
    boundBone {boneName=dyn_hair_strand_b_02}
    componentAppliedLinearAccClamp=<3000 3000 3000>
    componentLinearAccScale=<0.3 0.3 0.3>
    componentLinearVelScale=<0.3 0.3 0.3>
    gravityOverride {z=-950}
    physicsBodyDefinitions [{
      boundBone {boneName=dyn_hair_strand_b_02}
      boxExtents=<13 13 13>
      constraintSetup {angularLimitsMax=<5 5 20>, angularLimitsMin=<-5 -20 -20>, angularTarget=<0 0 0>}
      localJointOffset=<2 0 0>
     }]
   }
   n35 = Anim Dynamics - Bone: dyn_l_earring(ComponentPose=@n34, Alpha=1, ExternalForce="0,0,0") {
    angularDampingOverride=0.99
    angularSpringConstant=1000
    bAngularSpring=true
    bGravityOverrideInSimSpace=true
    bOverrideAngularDamping=true
    bUseGravityOverride=true
    alphaBoolBlend {blendOption=Linear}
    boundBone {boneName=dyn_l_earring}
    componentAppliedLinearAccClamp=<3000 3000 3000>
    componentLinearAccScale=<0.3 0.3 0.3>
    componentLinearVelScale=<0.3 0.3 0.3>
    gravityOverride {z=-980}
    physicsBodyDefinitions [{
      boundBone {boneName=dyn_l_earring}
      boxExtents=<12 12 12>
      constraintSetup {angularLimitsMax=<5 0 5>, angularLimitsMin=<-5 -10 -70>, angularTarget=<0 0 0>}
      localJointOffset=<3 0 0>
     }]
   }
   n36 = Anim Dynamics - Bone: dyn_r_earring(ComponentPose=@n35, Alpha=1, ExternalForce="0,0,0") {
    angularDampingOverride=0.99
    angularSpringConstant=1000
    bAngularSpring=true
    bGravityOverrideInSimSpace=true
    bOverrideAngularDamping=true
    bUseGravityOverride=true
    bUseSphericalLimits=true
    alphaBoolBlend {blendOption=Linear}
    boundBone {boneName=dyn_r_earring}
    componentAppliedLinearAccClamp=<3000 3000 3000>
    componentLinearAccScale=<0.3 0.3 0.3>
    componentLinearVelScale=<0.3 0.3 0.3>
    gravityOverride {z=-980}
    physicsBodyDefinitions [{
      collisionType=CustomSphere
      sphereCollisionRadius=2
      boundBone {boneName=dyn_r_earring}
      boxExtents=<12 12 12>
      constraintSetup {angularLimitsMax=<5 0 70>, angularLimitsMin=<-5 -10 0>, angularTarget=<0 0 0>}
      localJointOffset=<3 0 0>
     }]
    sphericalLimits [
     {limitRadius=7.3, drivingBone {boneName=upperarm_r}, sphereLocalOffset=<-3 0 0>}
     {limitRadius=5, drivingBone {boneName=head}, sphereLocalOffset=<0 0 0>}
    ]
   }
   n37 = Anim Dynamics - Bone: dyn_curl_a(ComponentPose=@n36, Alpha=1, ExternalForce="0,0,0") {
    angularDampingOverride=1
    angularSpringConstant=3500
    bAngularSpring=true
    bOverrideAngularDamping=true
    bUseGravityOverride=true
    bUsePlanarLimit=false
    gravityScale=0.35
    simulationSpace=RootRelative
    alphaBoolBlend {blendOption=Linear}
    boundBone {boneName=dyn_curl_a}
    componentAppliedLinearAccClamp=<1000 1000 1000>
    componentLinearAccScale=<0.6 0.5 0.5>
    componentLinearVelScale=<0.6 0.5 0.5>
    gravityOverride {z=-980}
    physicsBodyDefinitions [{
      boundBone {boneName=dyn_curl_a}
      boxExtents=<15 15 15>
      constraintSetup {angularLimitsMax=<10 10 5>, angularLimitsMin=<-10 -10 -30>, angularTarget=<1 0 0>}
      localJointOffset=<1 0 0>
     }]
   }
   n38 = Anim Dynamics - Bone: dyn_curl_b(ComponentPose=@n37, Alpha=1, ExternalForce="0,0,0") {
    angularDampingOverride=1
    angularSpringConstant=3500
    bAngularSpring=true
    bOverrideAngularDamping=true
    bUseGravityOverride=true
    bUsePlanarLimit=false
    gravityScale=0.35
    simulationSpace=RootRelative
    alphaBoolBlend {blendOption=Linear}
    boundBone {boneName=dyn_curl_b}
    componentAppliedLinearAccClamp=<1000 1000 1000>
    componentLinearAccScale=<0.6 0.5 0.5>
    componentLinearVelScale=<0.6 0.5 0.5>
    gravityOverride {z=-980}
    physicsBodyDefinitions [{
      boundBone {boneName=dyn_curl_b}
      boxExtents=<13 13 13>
      constraintSetup {angularLimitsMax=<10 20 10>, angularLimitsMin=<-10 -20 -30>, angularTarget=<1 0 0>}
      localJointOffset=<2 0 0>
     }]
   }
   n39 = Anim Dynamics - Bone: dyn_curl_c(ComponentPose=@n38, Alpha=1, ExternalForce="0,0,0") {
    angularDampingOverride=1
    angularSpringConstant=3500
    bAngularSpring=true
    bOverrideAngularDamping=true
    bUseGravityOverride=true
    bUsePlanarLimit=false
    gravityScale=0.35
    simulationSpace=RootRelative
    alphaBoolBlend {blendOption=Linear}
    boundBone {boneName=dyn_curl_c}
    componentAppliedLinearAccClamp=<1000 1000 1000>
    componentLinearAccScale=<0.6 0.5 0.5>
    componentLinearVelScale=<0.6 0.5 0.5>
    gravityOverride {z=-980}
    physicsBodyDefinitions [{
      boundBone {boneName=dyn_curl_c}
      boxExtents=<11 11 11>
      constraintSetup {angularLimitsMax=<8 15 10>, angularLimitsMin=<-8 -15 -30>, angularTarget=<1 0 0>}
      localJointOffset=<1.5 0 0>
     }]
   }
   Save cached pose 'AnimDyn_Pose'(Pose=Component To Local(ComponentPose=@n39)) {cacheName=AnimDyn_Pose}
   n40 = Control Rig!(Source=Copy Pose From Mesh() {bUseAttachedParent=true}, Alpha, is altskin on:double=Get altskin on() -> (altskin on:double)) -> (Pose) {
    alphaBoolBlend {blendOption=Linear}
    controlRigAssetReference {blueprintRigClass=/Game/Characters/Player/Female/Medium/Bodies/F_MED_StrongHibiscus/Meshes/Parts/F_MED_StrongHibiscus_FaceAcc_CtrlRig.F_MED_StrongHibiscus_FaceAcc_CtrlRig_C}
   }
   n41 = "Transform (Modify) Bone - Bone: upperarm_l"!(ComponentPose=Local To Component(LocalPose=@n40), Alpha=1, Rotation="0,0,0") -> (Pose) {rotationMode=BMM_Replace, rotationSpace=BCS_WorldSpace, alphaBoolBlend {blendOption=Linear}, boneToModify {boneName=upperarm_l}}
   n42 = "Transform (Modify) Bone - Bone: upperarm_r"!(ComponentPose=@n41, Alpha=1, Rotation="0,0,0") -> (Pose) {rotationMode=BMM_Replace, rotationSpace=BCS_WorldSpace, alphaBoolBlend {blendOption=Linear}, boneToModify {boneName=upperarm_r}}
   n43 = "Transform (Modify) Bone - Bone: upperarm_twist_01_l"!(ComponentPose=@n42, Alpha=1, Rotation="0,0,0") -> (Pose) {rotationMode=BMM_Replace, rotationSpace=BCS_WorldSpace, alphaBoolBlend {blendOption=Linear}, boneToModify {boneName=upperarm_twist_01_l}}
   n44 = "Transform (Modify) Bone - Bone: upperarm_twist_01_r"!(ComponentPose=@n43, Alpha=1, Rotation="0,0,0") -> (Pose) {rotationMode=BMM_Replace, rotationSpace=BCS_WorldSpace, alphaBoolBlend {blendOption=Linear}, boneToModify {boneName=upperarm_twist_01_r}}
   Save cached pose 'Start_Pose'(Pose=Component To Local(ComponentPose=@n44)) {cacheName=Start_Pose}
  ]}
 EventGraph {nodes [
   Event BlueprintUpdateAnimation()
   > n1 = Sequence()
   > n2 = Gravity Override State Function DEPRECATED(CharacterPartAnimInstance=Self-Reference(), GameTimeInSeconds=GetGameTimeInSeconds(), GravityOverrideParams=Get Gravity Override Params())
   > Set Gravity Override(Gravity Override=@n2)
   n3 = Get BodyMesh() -> (BodyMesh:Skeletal Mesh Component Object Reference)
   n4 = BreakTransform(InTransform=GetSocketTransform(self=@n3, InSocketName=upperarm_twist_01_l, TransformSpace=RTS_World))
   > Set Upperarm_Twist_L_Translation(Upperarm_Twist_L_Translation:Vector=@n4) -> (Output_Get:Vector)
   > Set Upperarm_Twist_L_Rotation(Upperarm_Twist_L_Rotation:Rotator=@n4.Rotation) -> (Output_Get:Rotator)
   n5 = BreakTransform(InTransform=GetSocketTransform(self=@n3, InSocketName=upperarm_twist_01_r, TransformSpace=RTS_World))
   > Set Upperarm_Twist_R_Translation(Upperarm_Twist_R_Translation:Vector=@n5) -> (Output_Get:Vector)
   > Set Upperarm_Twist_R_Rotation(Upperarm_Twist_R_Rotation:Rotator=@n5.Rotation) -> (Output_Get:Rotator)
   > Set Upperarm_L_Rotation(Upperarm_L_Rotation:Rotator=BreakTransform(InTransform=GetSocketTransform(self=@n3, InSocketName=upperarm_l, TransformSpace=RTS_World)).Rotation) -> (Output_Get:Rotator)
   > Set Upperarm_R_Rotation(Upperarm_R_Rotation:Rotator=BreakTransform(InTransform=GetSocketTransform(self=@n3, InSocketName=upperarm_r, TransformSpace=RTS_World)).Rotation) -> (Output_Get:Rotator)
   n6 = Get bRigidBodyEnabled()
   > Set WithoutRB(WithoutRB:bool=OR Boolean(A=Get isWearingBack() -> (isWearingBack:bool), B=NOT Boolean(A=@n6))) -> (Output_Get:bool)
   > Set InSkyWithBack(InSkyWithBack:bool=AND Boolean(A=Get isWearingBack(), B=Get bIsSkyDiving())) -> (Output_Get:bool)
   n7 = Get bRigidBodyEnabled()
   Set altskin on(execute=@n1.then_1, altskin on:double="To Float (Boolean)"(InBool=NOT Boolean(A=@n7))) -> (Output_Get:double)
   n8 = Event OnCharacterCustomizationLoaded()
   > Set BodyMesh(BodyMesh=Get Array Item with EFort Custom Part Type(Array=@n8.CustomMeshParts))
   n9 = CheckForBackbling()
   > n10 = Set isWearingBack(isWearingBack:bool=@n9) -> (Output_Get:bool)
   > Set isWearingCape(isWearingCape=@n9.bIsBackblingACape)
   > Set WithoutBack(WithoutBack:bool=NOT Boolean(A=@n10)) -> (Output_Get:bool)
  ]}
]
variables [
 altskin on {type=number}
 BodyMesh {type=SkeletalMeshComponent}
 Gravity Override {type=Vector}
 Gravity Override Params {type=GravityOverrideParamsStruct}
 InSkyWithBack {type=boolean}
 isWearingBack {type=boolean}
 isWearingCape {type=boolean}
 Upperarm_L_Rotation {type=Rotator}
 Upperarm_R_Rotation {type=Rotator}
 Upperarm_Twist_L_Rotation {type=Rotator}
 Upperarm_Twist_L_Translation {type=Vector}
 Upperarm_Twist_R_Rotation {type=Rotator}
 Upperarm_Twist_R_Translation {type=Vector}
 WithoutBack {type=boolean}
 WithoutRB {type=boolean}
]
